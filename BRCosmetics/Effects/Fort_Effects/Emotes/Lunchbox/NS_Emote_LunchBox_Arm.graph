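# Niagara system for the Lunchbox emote's arm effect: a golden energy charge that follows the user-supplied skeletal mesh.
# LightningArcCharge draws flickering ribbons rotating around bones, using partition modules and FN_BoneFlow for placement.
# Flare_and_Light moves a growing sprite and light along the bone flow as it ages, while Decal pins two animated-material sprites to bones or sockets.
# EnergyAbsorb, which pulls sparks toward a sampled point, is disabled.
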NiagaraSystem /BRCosmetics/Effects/Fort_Effects/Emotes/Lunchbox/NS_Emote_LunchBox_Arm
dependencies {
 usedDataInterfaces [&NiagaraDataInterfaceColorCurve, &NiagaraDataInterfaceCurve]
 usedDynamicInputs []
 usedModules [&SystemState]
}
emitters {
 Decal {
  emitterUpdateScript [
   &EmitterState()
   &SpawnBurst_Instantaneous(Spawn Count=2, Use Spawn Probability=false)
  ]
  particleSpawnScript [
   &FN_SetSpawnGroup_v2(Number of Spawn Groups=2)
   /Niagara/Modules/Spawn/Initialization/InitializeParticle(
    Color=<rgba 1 0.9637 0.5573 0.8>
    Module.Write Color=true
    Module.Write Lifetime=true
    Module.Write Mass=true
    Module.Write Material Random=true
    Module.Write Position=false
    Module.Write Scale=false
    Module.Write SpriteRotation=false
    Module.Write SpriteSize=true
    Position=&SimulationPosition()
    Sprite Size=&Vector2DFromFloat(Value=26)
   )
   &DynamicMaterialParameters(
    Index 0 Param 1=0.5
    Index 0 Param 2=1
    Index 0 Param 3=1
    Module.Param0WriteEnabled=false
    Module.Param1WriteEnabled=true
    Module.Param2WriteEnabled=true
    Module.Param3WriteEnabled=true
   )
   &CameraOffset(Camera Offset Amount=8)
  ]
  particleUpdateScript [
   &SkeletalMeshLocation(
    Apply Sampled Orientation to Mesh Particles=true
    Bone / Socket Index=&Add_Integer(A=$Particles.SpawnGroup, B=5)
    Bone Sampling Mode=::NewEnumerator4
    Inherit Velocity=false
    MaskSpawnGroup=false
    Skeletal Mesh=$User.SkeletalMesh
   )
   &SolveForcesAndVelocity(Clamp Velocity=false, Limit Acceleration=false)
   &ParticleState()
   &DynamicMaterialParameters(
    Index 0 Param 1=&FloatFromCurve(FloatCurve=DataInterface() {propertyValues {Curve {keys [<at 0 8>,<at 0.4 0.1>,<at 0.65 0.1>,<at 1 10>]}}})
    Module.Param0WriteEnabled=true
    Module.Param1WriteEnabled=false
    Module.Param2WriteEnabled=false
    Module.Param3WriteEnabled=false
   ) {module=DynamicMaterialParameters001}
   &ScaleColor(
    Scale RGB=&VectorFromFloat(Value=&FloatFromCurve(FloatCurve=DataInterface() {propertyValues {Curve {keys [<at 0.5 1>,<at 1 2>]}}}))
    ScaleA=true
    ScaleRGB=true
   )
  ]
  rendererClasses [&NiagaraSpriteRendererProperties]
  renderers [{rendererClass=&NiagaraSpriteRendererProperties}]
 }
 EnergyAbsorb {
  bEnabled=false
  emitterUpdateScript [
   &EmitterState()
   &SpawnRate(
    Spawn Probability=&FloatFromCurve(FloatCurve=DataInterface() {propertyValues {Curve {keys [<at 0 1 RCIM_Constant>,<at 0.2 0>]}}}, CurveIndex=$System.NormalizedLoopAge)
    SpawnRate=200
    Use Spawn Probability=true
   )
   &SpawnBurst_Instantaneous(Spawn Count=50, Use Spawn Probability=false)
  ]
  particleSpawnScript [
   &InitializeParticle(
    Color=&RandomRangeLinearColor(Minimum=<rgba 16 13.6927 4.8333 1>, Maximum=<rgba 20 18.852 9.375 1>)
    Module.Write Color=true
    Module.Write Lifetime=true
    Module.Write Mass=false
    Module.Write Material Random=true
    Module.Write Position=false
    Module.Write Scale=false
    Module.Write SpriteRotation=false
    Module.Write SpriteSize=true
    Position=&SimulationPosition()
    Sprite Size=&Vector2DFromFloat(Value=&RandomRangeFloat(Minimum=1.5, Maximum=2.25))
   )
   &SphereLocation(
    MaskSpawnGroup=false
    Override Seed=false
    Sphere Radius=200
    Surface Only=false
    UseNonUniformScale=false
    UseOffset=false
    UseSphereOrientation=false
   )
   &AddVelocity(Velocity=&RandomRangeVector(Minimum=<-100 -100 -100>, Maximum=<100 100 100>))
  ]
  particleUpdateScript [
   &ParticleState()
   &SkeletalMeshLocation(
    Bone / Socket Index=6
    Bone Sampling Mode=::NewEnumerator4
    Inherit Velocity=false
    MaskSpawnGroup=false
    Sampled Position Offset:NiagaraPosition=$Particles.Initial.Position
    Skeletal Mesh=$User.SkeletalMesh
   )
   /Niagara/Modules/Update/Forces/PointAttractionForce(
    Attraction Radius=225
    AttractionStrength=150
    AttractorPosition:Vector3f=$Output.SkeletalMeshLocation.SampledPosition
    Kill Radius:NiagaraFloat=&Max_Float(
     A:NiagaraFloat=5
     B:NiagaraFloat=&VectorLength(Vector=&Multiply_VectorByFloat(Vector=$Particles.Velocity, Float=$Engine.DeltaTime))
    )
    Module.Kill Within Radius=true
    Module.Use Falloff=false
   )
   &SolveForcesAndVelocity(Clamp Velocity=false, Limit Acceleration=false)
   &ScaleSpriteSizeBySpeed(Max Scale Factor=<1 2>, Min Scale Factor=<0.75 1.5>)
   &ScaleColor(
    Scale Alpha=&FloatFromCurve(FloatCurve=DataInterface() {propertyValues {Curve {keys [<at 0.2 0 RCIM_Cubic>,<at 0.5 1>,<at 0.95 1 RCIM_Cubic>,<at 1 0>]}}})
    Scale RGB=&MakeVectorFromLinearColorRGB(Color=&ColorFromCurve(ColorCurve=DataInterface() {propertyValues {
        AlphaCurve {keys [<at 0 1 RCIM_Cubic>]}
        BlueCurve {keys [<at 0.8 3 RCIM_Cubic>,<at 0.95 0 RCIM_Cubic>]}
        GreenCurve {keys [<at 0.8 3 RCIM_Cubic>,<at 0.95 4.7017 RCIM_Cubic>]}
        RedCurve {keys [<at 0.8 3 RCIM_Cubic>,<at 0.95 8 RCIM_Cubic>]}
       }}))
    ScaleA=true
    ScaleRGB=true
   )
  ]
  rendererClasses [&NiagaraSpriteRendererProperties]
  renderers [{rendererClass=&NiagaraSpriteRendererProperties}]
 }
 Flare_and_Light {
  emitterSpawnScript [
   /SetVariables(Emitter.RandomSeed:NiagaraInt32=/Niagara/DynamicInputs/Random/RandomInt(Randomness Mode:ENiagaraRandomnessMode=::NewEnumerator0, Random Seed:NiagaraInt32=0, Fixed Random Seed:NiagaraBool=false)) {isSetParameters=true, module=SetVariables_#1}
   /Game/Effects/Niagara/Modules/Partition/FN_PartitionAttributes(Number of Bone Sets=1)
  ]
  emitterUpdateScript [
   &EmitterState()
   &SpawnBurst_Instantaneous(Spawn Count=$Emitter.FN_PartitionAttributes.Spawn Total, Use Spawn Probability=false)
  ]
  particleSpawnScript [
   /Game/Effects/Niagara/Modules/Partition/FN_ParticlePartition_v2/FN_Particle_Partition(Enable Ribbon ID=true, Seed:NiagaraInt32=$Emitter.RandomSeed)
   &InitializeParticle(
    Color=<rgba 1 1 1 0.25>
    Lifetime=0.7
    Module.Write Color=true
    Module.Write Lifetime=true
    Module.Write Mass=false
    Module.Write Material Random=true
    Module.Write Position=false
    Module.Write Scale=false
    Module.Write SpriteRotation=false
    Module.Write SpriteSize=true
    Position=&SimulationPosition()
    Sprite Size=&Vector2DFromFloat(Value=50)
   )
   &CameraOffset(Camera Offset Amount=14)
  ]
  particleUpdateScript [
   &ParticleState()
   &FN_BoneFlow(
    Height:NiagaraFloat=$Particles.NormalizedAge
    Interpolate Mesh Sampling=false
    SkeletalMesh:NiagaraDataInterfaceSkeletalMesh=$User.SkeletalMesh
    Write Position Attribute=true
   )
   &ScaleColor(
    Scale Alpha=&FloatFromCurve(FloatCurve=DataInterface() {propertyValues {Curve {keys [<at 0.3 0 RCIM_Cubic>, <at 0.7 1>, {arriveTangent=-3.3333, interpMode=RCIM_Cubic, leaveTangent=-3.3333, time=0.95, value=1}, <at 1 0>]}}})
    Scale RGB=&MakeVectorFromLinearColorRGB(Color=&ColorFromCurve(ColorCurve=DataInterface() {propertyValues {
        AlphaCurve {keys [<at 0 1 RCIM_Cubic>]}
        BlueCurve {keys [<at 0.8 1 RCIM_Cubic>,<at 0.95 4.1667 RCIM_Cubic>]}
        GreenCurve {keys [<at 0.8 1 RCIM_Cubic>,<at 0.95 7.7776 RCIM_Cubic>]}
        RedCurve {keys [<at 0.8 1 RCIM_Cubic>,<at 0.95 8 RCIM_Cubic>]}
       }}))
    ScaleA=true
    ScaleRGB=true
   )
   &DynamicMaterialParameters(
    Index 0 Param 1=0.5
    Index 0 Param 2=&FloatFromCurve(FloatCurve=DataInterface() {propertyValues {Curve {keys [<at 0.55 0.25 RCIM_Cubic>,<at 0.85 2>]}}})
    Index 0 Param 4=&FloatFromCurve(FloatCurve=DataInterface() {propertyValues {Curve {keys [<at 0.4 0 RCIM_Cubic>, {interpMode=RCIM_Cubic, leaveTangent=6.6667, time=0.5, value=1}, <at 0.8 2>]}}})
    Module.Param0WriteEnabled=true
    Module.Param1WriteEnabled=true
    Module.Param2WriteEnabled=true
    Module.Param3WriteEnabled=true
   )
   &ScaleSpriteSize(Scale Factor=&Vector2DFromFloat(Value=&FloatFromCurve(FloatCurve=DataInterface() {propertyValues {Curve {keys [<at 0.25 0>,<at 0.8 1>]}}})))
   &FN_Light_v3(
    Color=<rgba 8 6.7326 1.1667 1>
    Create Color Attribute=true
    Create Exponent Attribute=true
    Create Radius Attribute=true
    Radius:NiagaraFloat=&Multiply_Float(A=150, B=$Particles.NormalizedAge)
    Scale Color:NiagaraFloat=&Multiply_Float(A=2, B=$Particles.NormalizedAge)
   )
  ]
  rendererClasses [&NiagaraSpriteRendererProperties, &NiagaraLightRendererProperties]
  renderers [
   {rendererClass=&NiagaraSpriteRendererProperties}
   {rendererClass=&NiagaraLightRendererProperties, rendererIndex=1}
  ]
 }
 LightningArcCharge {
  emitterSpawnScript [
   /SetVariables(Emitter.RandomSeed:NiagaraInt32=&RandomInt(Randomness Mode:ENiagaraRandomnessMode=::NewEnumerator0, Random Seed:NiagaraInt32=0, Fixed Random Seed:NiagaraBool=false)) {isSetParameters=true, module=SetVariables_#1}
   &FN_PartitionAttributes(Number of Bone Sets=2, Particles Per Partition=8, Partitions=2)
  ]
  emitterUpdateScript [
   &EmitterState()
   &SpawnRate(SpawnRate=20, Use Spawn Probability=false)
  ]
  particleSpawnScript [
   &FN_Particle_Partition(Enable Ribbon ID=true, Pattern:ENiagaraPartitionWaveForm=::NewEnumerator1, Seed:NiagaraInt32=$Emitter.RandomSeed)
   &InitializeRibbon(
    Color=<rgba 6 4.9108 0 1>
    Lifetime=0.75
    Module.Write Color=true
    Module.Write Lifetime=true
    Module.Write Mass=true
    Module.Write Material Random=true
    Module.Write Position=true
    Module.Write RibbonTwist=false
    Module.Write RibbonWidth=true
    Ribbon Width=20
   )
   /Game/Effects/Niagara/Modules/BoneFlow/FN_Initialize_BoneFlow_Attributes_v2(Angle:NiagaraFloat=&UniformRangedFloat(Minimum=0, Maximum=1), Radius=10)
   &CameraOffset(Camera Offset Amount=5)
  ]
  particleUpdateScript [
   &ParticleState()
   &FN_BoneFlow(
    Angle:NiagaraFloat=&Add_Float(A=$Particles.Initial.FN_Initialize_BoneFlow_Attributes_v2.Angle, B=HlslExpression() {hlslExpression=Emitter.Age * .2})
    Height:NiagaraFloat=$Particles.Partition.NormalizedID
    Interpolate Mesh Sampling=false
    Radius:NiagaraFloat=$Particles.Initial.FN_Initialize_BoneFlow_Attributes_v2.Radius
    SkeletalMesh:NiagaraDataInterfaceSkeletalMesh=$User.SkeletalMesh
    Write Position Attribute=true
   )
   &Color(
    Color=&ColorFromCurve(ColorCurve=DataInterface() {propertyValues {
       AlphaCurve {keys [<at 0 0 RCIM_Cubic>,<at 0.1239 1>,<at 0.346 1>,<at 1 0 RCIM_Cubic>]}
       BlueCurve {keys [<at -0.0056 2>,<at 0.261 0.2>]}
       GreenCurve {keys [<at -0.0056 7.6765>,<at 0.261 0.7676>]}
       RedCurve {keys [<at -0.0056 10>,<at 0.261 1>]}
      }})
    Scale Alpha=&FloatFromCurve(FloatCurve=DataInterface() {propertyValues {Curve {keys [<at 0.1282 1.0115>,<at 0.2455 0.0123>,<at 0.3389 0.9565>,<at 0.4758 0.9657>,<at 0.591 0.0032>,<at 0.6388 1.0115>,<at 0.8061 1>,<at 0.95 1>,<at 1 0>]}}})
    Scale Color=&VectorFromFloat(Value=&FloatFromCurve(FloatCurve=DataInterface() {propertyValues {Curve {keys [<at 0 4>,<at 0.2994 0>,<at 0.573 2>,<at 0.8425 0.0123>]}}}))
   )
   &DynamicMaterialParameters(
    Index 0 Param 2=&FloatFromCurve(FloatCurve=DataInterface() {propertyValues {Curve {keys [<at 0 0 RCIM_Cubic>, {arriveTangent=-6.7156, interpMode=RCIM_Cubic, leaveTangent=-6.7156, tangentMode=RCTM_User, time=1, value=-2}]}}})
    Module.Param0WriteEnabled=true
    Module.Param1WriteEnabled=true
    Module.Param2WriteEnabled=true
    Module.Param3WriteEnabled=true
   )
  ]
  rendererClasses [&NiagaraRibbonRendererProperties]
  renderers [{rendererClass=&NiagaraRibbonRendererProperties}]
 }
}
summary {
 systemName=NS_Emote_LunchBox_Arm
 emitters [
  {emitterName=LightningArcCharge, rendererClasses [&NiagaraRibbonRendererProperties]}
  {emitterName=Decal, rendererClasses [&NiagaraSpriteRendererProperties]}
  {
   emitterName=Flare_and_Light
   rendererClasses [&NiagaraSpriteRendererProperties, &NiagaraLightRendererProperties]
  }
  {bEnabled=false, emitterName=EnergyAbsorb, rendererClasses [&NiagaraSpriteRendererProperties]}
 ]
 userVariables [User.SkeletalMesh {
   type=NiagaraDataInterfaceSkeletalMesh
   defaultValue {
    struct=NiagaraExt_VariableValue_DataInterface
    value {dataInterface=&self:NiagaraDataInterfaceSkeletalMesh_0}
   }
  }]
}
