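# Niagara system for the LunchBox emote impact: a short, golden-yellow electrical burst with no user parameters.
# Independent burst emitters layer ground-facing electric and smoke sprites, radial lightning sprites, lightning residue meshes, falling sparks, a mesh spike with sprite flare, and a brief light.
# Age curves drive scale, color fades and dynamic material parameters for animated shader effects.
# The Electric_Strokes and ImpactShockWave_Energy emitters are disabled.
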NiagaraSystem /BRCosmetics/Effects/Fort_Effects/Emotes/Lunchbox/NS_Emote_LunchBox_Impact
dependencies {
 usedDataInterfaces [&NiagaraDataInterfaceCurve, &NiagaraDataInterfaceVectorCurve, &NiagaraDataInterfaceColorCurve]
 usedDynamicInputs []
 usedModules [&SystemState]
}
emitters {
 ElectricGround {
  emitterUpdateScript [
   &EmitterState()
   &SpawnBurst_Instantaneous(Spawn Count=2, Use Spawn Probability=false)
  ]
  particleSpawnScript [
   /Niagara/Modules/Spawn/Initialization/InitializeParticle(
    Color=<rgba 1.25 1.0907 0.5778 1>
    Lifetime=&UniformRangedFloat(Minimum=0.4, Maximum=0.5)
    Module.Write Color=true
    Module.Write Lifetime=true
    Module.Write Mass=true
    Module.Write Material Random=true
    Module.Write Position=true
    Module.Write Scale=false
    Module.Write SpriteRotation=true
    Module.Write SpriteSize=true
    Sprite Rotation=&RandomRangeFloat(Minimum=-1, Maximum=1)
    Sprite Size=&Vector2DFromFloat(Value=&RandomRangeFloat(Minimum=75, Maximum=100))
   )
   &DynamicMaterialParameters(
    Index 0 Param 1=&RandomRangeFloat(Minimum=0, Maximum=1)
    Module.Param0WriteEnabled=true
    Module.Param1WriteEnabled=false
    Module.Param2WriteEnabled=false
    Module.Param3WriteEnabled=false
   )
   &CameraOffset(Camera Offset Amount=20)
   /SetVariables(Particles.UVScale:Vector2f=&MakeVector2D(X=&UniformAOrBFloat(A=-1), Y=&UniformAOrBFloat(A=-1))) {isSetParameters=true, module=SetVariables_#1}
   /NiagaraFN/Sprite/FN_SpriteAlignmentAndFacing(SpriteAlignment:Vector3f=<0 0 1>, SpriteFacing:Vector3f=<0 0 1>) {module=SpriteAlignmentAndFacing}
  ]
  particleUpdateScript [
   &ParticleState()
   &SolveForcesAndVelocity(Clamp Velocity=false, Limit Acceleration=false)
   &ScaleSpriteSize(Scale Factor=&Vector2DFromFloat(Value=&FloatFromCurve(FloatCurve=DataInterface() {propertyValues {Curve {keys [{arriveTangent=1.5318, interpMode=RCIM_Cubic, leaveTangent=1.5318, tangentMode=RCTM_User, time=0, value=0.35}, <at 1 1>]}}})))
  ]
  rendererClasses [&NiagaraSpriteRendererProperties]
  renderers [{rendererClass=&NiagaraSpriteRendererProperties}]
 }
 Electric_Strokes {
  bEnabled=false
  emitterUpdateScript [
   &EmitterState()
   &SpawnBurst_Instantaneous(Spawn Count=&UniformRangedInt(Minimum=2, Maximum=3), Use Spawn Probability=false)
  ]
  particleSpawnScript [
   &InitializeParticle(
    Color=<rgba 1.25 1.0934 0.2865 1>
    Lifetime=&UniformRangedFloat(Minimum=0.225, Maximum=0.275)
    Module.Write Color=true
    Module.Write Lifetime=true
    Module.Write Mass=true
    Module.Write Material Random=true
    Module.Write Position=true
    Module.Write Scale=false
    Module.Write SpriteRotation=true
    Module.Write SpriteSize=true
    Sprite Rotation=&RandomRangeFloat(Minimum=-1, Maximum=1)
    Sprite Size=&MakeVector2D(
     X=&UniformRangedFloat(Minimum=150, Maximum=175)
     Y=&UniformRangedFloat(Minimum=140, Maximum=200)
    )
   )
   &DynamicMaterialParameters(
    Index 0 Param 1=&UniformRangedFloat(Minimum=2.5, Maximum=3.5)
    Index 0 Param 3=&UniformRangedFloat(Minimum=0.2, Maximum=0.5)
    Module.Param0WriteEnabled=true
    Module.Param1WriteEnabled=false
    Module.Param2WriteEnabled=true
    Module.Param3WriteEnabled=true
   )
   &CameraOffset(Camera Offset Amount=20)
   /Niagara/Modules/Spawn/Location/SphereLocation(Hemisphere Z=true, MaskSpawnGroup=false, Module.Override Randomness=false, Sphere Radius=10, Surface Only=false)
   &AddVelocity(Velocity=&UniformRangedVector(Minimum=<0 0 50>, Maximum=<0 0 200>))
   &AddVelocity(Velocity=&Multiply_Vector(A=$Output.SphereLocation.SphereNormal, B=&VectorFromFloat(Value=125))) {module=AddVelocity001}
   /SetVariables(Particles.UVScale:Vector2f=&MakeVector2D(X=&UniformAOrBFloat(A=-1), Y=&UniformAOrBFloat(A=-1))) {isSetParameters=true, module=SetVariables_#1}
  ]
  particleUpdateScript [
   &ParticleState()
   &SolveForcesAndVelocity(Clamp Velocity=false, Limit Acceleration=false)
   &ScaleSpriteSize(Scale Factor=&Vector2DFromFloat(Value=&FloatFromCurve(FloatCurve=DataInterface() {propertyValues {Curve {keys [{arriveTangent=1.9395, interpMode=RCIM_Cubic, leaveTangent=1.9395, tangentMode=RCTM_User, time=0, value=0.3}, <at 1 1 RCIM_Cubic>]}}})))
   &ScaleColor(Scale RGB=&VectorFromFloat(Value=5), ScaleA=true, ScaleRGB=true)
   &DynamicMaterialParameters(
    Index 0 Param 2=&FloatFromCurve(FloatCurve=DataInterface() {propertyValues {Curve {keys [<at 0 -0.1 RCIM_Cubic>, {arriveTangent=-1.6318, interpMode=RCIM_Cubic, leaveTangent=-1.6318, tangentMode=RCTM_User, time=1, value=-1}]}}})
    Module.Param0WriteEnabled=false
    Module.Param1WriteEnabled=true
    Module.Param2WriteEnabled=false
    Module.Param3WriteEnabled=false
   ) {module=DynamicMaterialParameters001}
  ]
  rendererClasses [&NiagaraSpriteRendererProperties]
  renderers [{rendererClass=&NiagaraSpriteRendererProperties}]
 }
 ImpactShockWave_Energy {
  bEnabled=false
  emitterUpdateScript [
   &EmitterState()
   &SpawnBurst_Instantaneous(Spawn Count=1, Use Spawn Probability=false)
  ]
  particleSpawnScript [
   &InitializeParticle(
    Color=<rgba 1 0.8399 0.1063 0.5>
    Lifetime=0.55
    Mesh Scale=<2.5 0.2 2.5>
    Module.Write Color=true
    Module.Write Lifetime=true
    Module.Write Mass=true
    Module.Write Material Random=true
    Module.Write Position=false
    Module.Write Scale=true
    Module.Write SpriteRotation=false
    Module.Write SpriteSize=false
    Position=&SimulationPosition()
   )
   /Niagara/Modules/Spawn/Orientation/V2/InitialMeshOrientation(
    Orientation Axis:Vector3f=<0 0 1>
    Orientation Vector=<0 1 0>
    Rotation=&MakeVector(X=0, Y=&UniformRangedFloat(Minimum=0, Maximum=1), Z=0)
    Use Orientation Vector=true
    Use Rotation Vector=true
   )
  ]
  particleUpdateScript [
   &ParticleState()
   /Game/Effects/Niagara/Modules/Mesh/FN_MeshRenderer_v3/FN_MeshRenderer_v2(
    Color:LinearColor=&MultiplyLinearColor(
     Linear Color=<rgba 5 5 5 1.5>
     Scale Factor=&MakeFloat4FromVectorAndFloat(
      "Vector (RGB)"=&MakeVectorFromLinearColorRGB(Color=&ColorFromCurve(ColorCurve=DataInterface() {propertyValues {
          AlphaCurve {keys [<at 0 1 RCIM_Cubic>,<at 1 1 RCIM_Cubic>]}
          BlueCurve {keys [<at 0.385 0 RCIM_Cubic>,<at 1 1 RCIM_Cubic>]}
          GreenCurve {keys [<at 0.385 0.8762 RCIM_Cubic>,<at 1 1 RCIM_Cubic>]}
          RedCurve {keys [<at 0.385 1 RCIM_Cubic>,<at 1 1 RCIM_Cubic>]}
         }}))
      "Float (Alpha)"=&FloatFromCurve(
       FloatCurve=DataInterface() {propertyValues {Curve {keys [
           {interpMode=RCIM_Cubic, tangentMode=RCTM_User, time=0, value=0}
           {arriveTangent=2.0824, interpMode=RCIM_Cubic, leaveTangent=0.0112, tangentMode=RCTM_Break, time=0.0553, value=0.9565}
           {arriveTangent=-0.1758, interpMode=RCIM_Cubic, leaveTangent=-0.1758, tangentMode=RCTM_User, time=1, value=0}
          ]}}}
       Scale Curve=0.5
      )
     )
    )
    Create COlor Attribute=true
    Create Orientation Attribute=true
    Create Scale Attribute=true
    Scale:Vector3f=&VectorFromFloat(Value=&Multiply_Float(
      A=3
      B=&FloatFromCurve(FloatCurve=DataInterface() {propertyValues {Curve {keys [{arriveTangent=2.4265, interpMode=RCIM_Cubic, leaveTangent=2.4265, tangentMode=RCTM_User, time=0, value=0}, <at 1 1>]}}})
     ))
   )
   &ScaleMeshSize(Scale Factor=/Niagara/DynamicInputs/ScaleByCurve/ScaleVectorByCurve(
     InitialValue:Vector3f=<1 1 1>
     ScaleCurve:NiagaraDataInterfaceVectorCurve=DataInterface() {propertyValues {
       XCurve {keys [
         {arriveTangent=4.2859, interpMode=RCIM_Cubic, leaveTangent=4.2859, tangentMode=RCTM_User, time=0, value=0}
         {arriveTangent=2, interpMode=RCIM_Cubic, leaveTangent=2, tangentMode=RCTM_User, time=0.4394, value=1.6134}
         {interpMode=RCIM_Cubic, tangentMode=RCTM_User, time=1, value=2}
        ]}
       YCurve {keys [
         {arriveTangent=5.8758, interpMode=RCIM_Cubic, leaveTangent=5.8759, tangentMode=RCTM_User, time=0, value=0}
         {arriveTangent=2, interpMode=RCIM_Cubic, leaveTangent=2, tangentMode=RCTM_User, time=0.4581, value=1.6134}
         {interpMode=RCIM_Cubic, tangentMode=RCTM_User, time=1, value=2.2}
        ]}
       ZCurve {keys [
         {arriveTangent=4.7007, interpMode=RCIM_Cubic, leaveTangent=4.7007, tangentMode=RCTM_User, time=0, value=0}
         {arriveTangent=2, interpMode=RCIM_Cubic, leaveTangent=2, tangentMode=RCTM_User, time=0.4519, value=1.549}
         {interpMode=RCIM_Cubic, tangentMode=RCTM_User, time=1, value=2}
        ]}
      }}
     CurveIndex:NiagaraFloat=$Particles.NormalizedAge
    ))
   &ScaleColor(
    Scale Alpha=&FloatFromCurve(FloatCurve=DataInterface() {propertyValues {Curve {keys [
         {interpMode=RCIM_Cubic, tangentMode=RCTM_User, time=0, value=1}
         {arriveTangent=-1.0036, interpMode=RCIM_Cubic, leaveTangent=-1.0036, tangentMode=RCTM_User, time=0.2417, value=0.2272}
         {arriveTangent=-0.1758, interpMode=RCIM_Cubic, leaveTangent=-0.1758, tangentMode=RCTM_User, time=0.95, value=0}
        ]}}})
    Scale RGB=&VectorFromCurve(VectorCurve=DataInterface() {propertyValues {
       XCurve {keys [{arriveTangent=-3.6966, interpMode=RCIM_Cubic, leaveTangent=-3.6966, tangentMode=RCTM_User, time=-0.0011, value=3}, <at 0.9826 0.318>]}
       YCurve {keys [<at -0.0011 2.5>,<at 0.9988 0>]}
       ZCurve {keys [<at -0.0011 2.5>,<at 0.9988 0>]}
      }})
    ScaleA=true
    ScaleRGB=true
   )
   &DynamicMaterialParameters(
    Index 0 Param 1=&FloatFromCurve(FloatCurve=DataInterface() {propertyValues {Curve {keys [{arriveTangent=1.5708, interpMode=RCIM_Cubic, leaveTangent=1.5708, tangentMode=RCTM_User, time=0, value=0.3}, <at 1 1.2>]}}})
    Index 0 Param 4=3
    Index 1 Param 1=&FloatFromCurve(FloatCurve=DataInterface() {propertyValues {Curve {keys [<at 0 2.5>,<at 1 0>]}}})
    Index 1 Param 2=&FloatFromCurve(FloatCurve=DataInterface() {propertyValues {Curve {keys [<at 0 2.5>,<at 1 0>]}}})
    Module.Param0WriteEnabled=true
    Module.Param0WriteEnabled1=true
    Module.Param1WriteEnabled=true
    Module.Param1WriteEnabled1=true
    Module.Param2WriteEnabled=true
    Module.Param2WriteEnabled1=false
    Module.Param3WriteEnabled=true
    Module.Param3WriteEnabled1=false
    Write Parameter Index 1=true
   )
  ]
  rendererClasses [&NiagaraMeshRendererProperties]
  renderers [
   {rendererClass=&NiagaraMeshRendererProperties, rendererIndex=0}
   {rendererClass=&NiagaraMeshRendererProperties, rendererIndex=1}
  ]
 }
 Lazy_Sparks {
  emitterUpdateScript [
   &EmitterState()
   &SpawnBurst_Instantaneous(Spawn Count=&UniformRangedInt(Minimum=14, Maximum=20), Use Spawn Probability=false)
  ]
  particleSpawnScript [
   &InitializeParticle(
    Color=<rgba 6 5.0404 1.875 1>
    Lifetime=&UniformRangedFloat(Minimum=0.75, Maximum=0.9)
    Module.Write Color=true
    Module.Write Lifetime=true
    Module.Write Mass=true
    Module.Write Material Random=true
    Module.Write Position=false
    Module.Write Scale=false
    Module.Write SpriteRotation=false
    Module.Write SpriteSize=true
    Position=&SimulationPosition()
    Sprite Size=&UniformRangedVector2D(Minimum=<1 1.5>, Maximum=<1.25 2.25>)
   )
   &SphereLocation(
    Hemisphere Z=true
    MaskSpawnGroup=false
    Module.Override Randomness=false
    Sphere Origin:NiagaraPosition=/Niagara/DynamicInputs/Helpers/SimulationPosition()
    Sphere Radius=40
    Surface Only=false
   )
   &AddVelocity(Velocity=&UniformRangedVector(Minimum=<-200 -200 250>, Maximum=<200 200 500>))
   &AddVelocity(Velocity=&Multiply_Vector(
     A=$Output.SphereLocation.SphereNormal
     B=&VectorFromFloat(Value=&UniformRangedFloat(Minimum=10, Maximum=30))
    )) {module=AddVelocity001}
  ]
  particleUpdateScript [
   &ParticleState()
   &AccelerationForce(Acceleration=<0 0 -1600>)
   &SolveForcesAndVelocity(Clamp Velocity=false, Limit Acceleration=false)
   &ScaleSpriteSizeBySpeed(Max Scale Factor=<1 3>, Min Scale Factor=<1 2>)
   &ScaleSpriteSize(Scale Factor=&Vector2DFromFloat(Value=&FloatFromCurve(FloatCurve=DataInterface() {propertyValues {Curve {keys [{arriveTangent=-6.725, interpMode=RCIM_Cubic, leaveTangent=-6.725, tangentMode=RCTM_User, time=0.65, value=1}, <at 1 0 RCIM_Cubic>]}}})))
   &ScaleColor(
    Scale RGB=&MakeVectorFromLinearColorRGB(Color=&ColorFromCurve(ColorCurve=DataInterface() {propertyValues {
        AlphaCurve {keys [<at 0 1 RCIM_Cubic>,<at 1 1 RCIM_Cubic>]}
        BlueCurve {keys [<at 0.2 10 RCIM_Cubic>,<at 0.7 0.5 RCIM_Cubic>]}
        GreenCurve {keys [<at 0.2 9.9741 RCIM_Cubic>,<at 0.7 0.5 RCIM_Cubic>]}
        RedCurve {keys [<at 0.2 9.375 RCIM_Cubic>,<at 0.7 0.5 RCIM_Cubic>]}
       }}))
    ScaleA=true
    ScaleRGB=true
   )
  ]
  rendererClasses [&NiagaraSpriteRendererProperties]
  renderers [{rendererClass=&NiagaraSpriteRendererProperties}]
 }
 Light {
  emitterUpdateScript [
   &EmitterState()
   &SpawnBurst_Instantaneous(Spawn Count=1, Use Spawn Probability=false)
  ]
  particleSpawnScript [&InitializeParticle(
    Lifetime=0.6
    Module.Write Color=true
    Module.Write Lifetime=true
    Module.Write Mass=true
    Module.Write Material Random=true
    Module.Write Position=false
    Module.Write Scale=false
    Module.Write SpriteRotation=true
    Module.Write SpriteSize=true
    Position=/Niagara/Functions/Localspace/SimulationPosition()
   )]
  particleUpdateScript [
   &ParticleState()
   &FN_Light_v3(
    Color=<rgba 10 8.7442 3.2292 1>
    Create Color Attribute=true
    Create Exponent Attribute=true
    Create Position Attribute=true
    Create Radius Attribute=true
    Exponent=6
    Position:Vector3f=&Add_Vector(A=Linked() {linkedVariable {name=Particles.Initial.Position, type=NiagaraPosition}}, B=<0 0 100>)
    Radius:NiagaraFloat=&Multiply_Float(
     A=125
     B=&FloatFromCurve(FloatCurve=DataInterface() {propertyValues {Curve {keys [{arriveTangent=7.6783, interpMode=RCIM_Cubic, leaveTangent=7.6783, tangentMode=RCTM_User, time=0, value=0}, <at 0.25 1>]}}})
    )
    Scale Color:NiagaraFloat=&FloatFromCurve(FloatCurve=DataInterface() {propertyValues {Curve {keys [<at 0.65 1 RCIM_Cubic>, {interpMode=RCIM_Cubic, leaveTangent=-6.6766, time=1, value=0}]}}})
   )
  ]
  rendererClasses [&NiagaraLightRendererProperties]
  renderers [{rendererClass=&NiagaraLightRendererProperties, rendererIndex=0}]
 }
 LightningSprites_Burst {
  emitterUpdateScript [
   &EmitterState()
   &SpawnBurst_Instantaneous(Spawn Count=&UniformRangedInt(Minimum=7, Maximum=10), Use Spawn Probability=false)
  ]
  particleSpawnScript [
   &InitializeParticle(
    Color=&RandomRangeLinearColor(Minimum=<rgba 2 1.9119 1.2708 0.5>, Maximum=<rgba 3 2.771 1.2031 1>)
    Lifetime=&RandomRangeFloat(Minimum=0.75, Maximum=1)
    Module.Write Color=true
    Module.Write Lifetime=true
    Module.Write Mass=true
    Module.Write Material Random=true
    Module.Write Position=false
    Module.Write Scale=false
    Module.Write SpriteRotation=true
    Module.Write SpriteSize=true
    Position=&SimulationPosition()
    Sprite Rotation=&UniformRangedFloat(Minimum=-1, Maximum=1)
    Sprite Size=&Vector2DFromFloat(Value=&RandomRangeFloat(Minimum=30, Maximum=40))
   )
   &DynamicMaterialParameters(
    Index 0 Param 1=&RandomRangeFloat(Minimum=0, Maximum=1)
    Module.Param0WriteEnabled=true
    Module.Param1WriteEnabled=true
    Module.Param2WriteEnabled=false
    Module.Param3WriteEnabled=false
   )
   &CylinderLocation(
    Cylinder Height=1
    Cylinder Radius=30
    MaskSpawnGroup=false
    Module.Enable Lathe Profile=false
    Module.Override Randomness=false
    Module.Surface Only=true
    Offset=<0 0 2>
    Surface Only Band Thickness=10
   )
   &AddVelocityFromPoint(Module.Use Falloff Distance=false, Velocity Strength=150)
   &SubUVAnimation(End Frame=3, SubUV Animation Mode:ENiagara_SubUVLookupMode=::NewEnumerator2)
  ]
  particleUpdateScript [
   &ParticleState()
   &Drag(Drag=8, Use Linear Drag=true, Use Rotational Drag=false)
   &SolveForcesAndVelocity(Clamp Velocity=false, Limit Acceleration=false)
   &DynamicMaterialParameters(
    Index 0 Param 3=&FloatFromCurve(FloatCurve=DataInterface() {propertyValues {Curve {keys [<at 0 1 RCIM_Cubic>,<at 0.2 0.025 RCIM_Cubic>,<at 1 1>]}}})
    Module.Param0WriteEnabled=false
    Module.Param1WriteEnabled=false
    Module.Param2WriteEnabled=true
    Module.Param3WriteEnabled=false
   ) {module=DynamicMaterialParameters001}
   &ScaleColor(
    Scale Alpha=&FloatFromCurve(FloatCurve=DataInterface() {propertyValues {Curve {keys [<at 0.75 1 RCIM_Cubic>, {arriveTangent=-10.3917, interpMode=RCIM_Cubic, leaveTangent=-10.3917, tangentMode=RCTM_User, time=1, value=0}]}}})
    Scale RGB=&Multiply_VectorByFloat(
     Vector=&MakeVectorFromLinearColorRGB(Color=&ColorFromCurve(ColorCurve=DataInterface() {propertyValues {
         AlphaCurve {keys []}
         BlueCurve {keys [<at 0 1>, {arriveTangent=-0.1146, interpMode=RCIM_Cubic, leaveTangent=-0.1146, time=0.15, value=4.0625}, <at 1 0.8854 RCIM_Cubic>]}
         GreenCurve {keys [<at 0 1>, {arriveTangent=-0.0202, interpMode=RCIM_Cubic, leaveTangent=-0.0202, time=0.15, value=4.9198}, <at 1 0.9798 RCIM_Cubic>]}
         RedCurve {keys [<at 0 1>,<at 0.15 5 RCIM_Cubic>,<at 1 1 RCIM_Cubic>]}
        }}))
     Float=1
    )
    ScaleA=true
    ScaleRGB=true
   )
  ]
  rendererClasses [&NiagaraSpriteRendererProperties]
  renderers [{rendererClass=&NiagaraSpriteRendererProperties}]
 }
 Lightning_residue {
  emitterUpdateScript [
   &EmitterState()
   &SpawnBurst_Instantaneous(Spawn Count=1, Use Spawn Probability=false)
   &SpawnBurst_Instantaneous(Spawn Count=2, Spawn Time=0.0458, Use Spawn Probability=false) {module=SpawnBurst_Instantaneous001}
  ]
  particleSpawnScript [
   &InitializeParticle(
    Lifetime=&UniformRangedFloat(Minimum=0.15, Maximum=0.225)
    Mesh Scale=&UniformRangedVector(Minimum=<2 2 4>, Maximum=<4 4 3>)
    Module.Write Color=true
    Module.Write Lifetime=true
    Module.Write Mass=true
    Module.Write Material Random=true
    Module.Write Position=false
    Module.Write Scale=true
    Module.Write SpriteRotation=false
    Module.Write SpriteSize=false
    Position=&SimulationPosition()
   )
   &InitialMeshOrientation(
    Rotation=&UniformRangedVector(Minimum=<-1 -1 -1>, Maximum=<1 1 1>)
    Use Orientation Vector=true
    Use Rotation Vector=true
   ) {module=InitialMeshOrientation001}
   &DynamicMaterialParameters(
    Index 0 Param 1=&UniformRangedFloat(Minimum=5, Maximum=10)
    Index 0 Param 3=&UniformRangedFloat(Minimum=0.2, Maximum=1)
    Index 0 Param 4=&UniformRangedFloat(Minimum=0, Maximum=1)
    Module.Param0WriteEnabled=true
    Module.Param1WriteEnabled=false
    Module.Param2WriteEnabled=true
    Module.Param3WriteEnabled=true
   )
   &SphereLocation(MaskSpawnGroup=false, Module.Override Randomness=false, Sphere Radius=10, Surface Only=false)
   &AddVelocity(Coordinate Space=::World, Velocity=<0 0 1>)
  ]
  particleUpdateScript [
   &ParticleState()
   &SolveForcesAndVelocity(Clamp Velocity=false, Limit Acceleration=false)
   &DynamicMaterialParameters(
    Index 0 Param 2=&FloatFromCurve(FloatCurve=DataInterface() {propertyValues {Curve {keys [{interpMode=RCIM_Cubic, leaveTangent=-2.846, tangentMode=RCTM_Break, time=0, value=1}, <at 1 -0.25>]}}})
    Module.Param0WriteEnabled=false
    Module.Param1WriteEnabled=true
    Module.Param2WriteEnabled=false
    Module.Param3WriteEnabled=false
   ) {module=DynamicMaterialParameters001}
   &ScaleMeshSize(Scale Factor=&MakeVector(
     X=1
     Y=1
     Z=&FloatFromCurve(FloatCurve=DataInterface() {propertyValues {Curve {keys [{arriveTangent=2.0159, interpMode=RCIM_Cubic, leaveTangent=2.0159, tangentMode=RCTM_User, time=0, value=0.25}, <at 1 1>]}}})
    ))
   &ScaleColor(
    Scale Alpha=&FloatFromCurve(FloatCurve=DataInterface() {propertyValues {Curve {keys [<at 0.25 1 RCIM_Cubic>, {arriveTangent=-3.8662, interpMode=RCIM_Cubic, leaveTangent=-3.8662, tangentMode=RCTM_User, time=1, value=0}]}}})
    ScaleA=true
    ScaleRGB=true
   )
  ]
  rendererClasses [&NiagaraMeshRendererProperties]
  renderers [{rendererClass=&NiagaraMeshRendererProperties, rendererIndex=0}]
 }
 SmokeGround {
  emitterUpdateScript [
   &EmitterState()
   &SpawnBurst_Instantaneous(Spawn Count=1, Use Spawn Probability=false)
  ]
  particleSpawnScript [
   &InitializeParticle(
    Color=&MultiplyLinearColor(
     Linear Color=<rgba 0.6094 0.5125 0 0.25>
     Scale Factor=&MakeVector4_FromLinearColor(LinearColor=<rgba 1 1 1 1>)
    )
    Lifetime=0.4
    Module.Write Color=true
    Module.Write Lifetime=true
    Module.Write Mass=true
    Module.Write Material Random=true
    Module.Write Position=false
    Module.Write Scale=false
    Module.Write SpriteRotation=true
    Module.Write SpriteSize=true
    Position=&SimulationPosition()
    Sprite Rotation=&UniformRangedFloat(Minimum=-180, Maximum=180)
    Sprite Size=&Vector2DFromFloat(Value=100)
   )
   /SetVariables(Particles.SpriteFacing=<0 0 1>) {isSetParameters=true}
   &SystemLocation(MaskSpawnGroup=false, Offset=<0 0 5>)
  ]
  particleUpdateScript [
   &ParticleState()
   &DynamicMaterialParameters(
    Index 0 Param 1=&FloatFromCurve(FloatCurve=DataInterface() {propertyValues {Curve {keys [<at 0 0.5>,<at 1 2>]}}})
    Module.Param0WriteEnabled=true
    Module.Param1WriteEnabled=false
    Module.Param2WriteEnabled=false
    Module.Param3WriteEnabled=false
   )
   &ScaleSpriteSize(Scale Factor=&Vector2DFromFloat(Value=&FloatFromCurve(FloatCurve=DataInterface() {propertyValues {Curve {keys [
          {arriveTangent=5.4041, interpMode=RCIM_Cubic, leaveTangent=5.4041, tangentMode=RCTM_User, time=0, value=0.4}
          {arriveTangent=0.6903, interpMode=RCIM_Cubic, leaveTangent=0.6903, tangentMode=RCTM_User, time=1, value=3.5}
         ]}}})))
   &CameraOffset(Camera Offset Amount=100)
   &ScaleColor(
    Scale Alpha=&FloatFromCurve(FloatCurve=DataInterface() {propertyValues {Curve {keys [
         <at 0 1 RCIM_Cubic>
         {arriveTangent=-1.4475, interpMode=RCIM_Cubic, leaveTangent=-1.4475, tangentMode=RCTM_User, time=0.8099, value=0.6059}
         {arriveTangent=-4.9981, interpMode=RCIM_Cubic, leaveTangent=-4.9981, tangentMode=RCTM_User, time=1.0011, value=0.002}
        ]}}})
    ScaleA=true
    ScaleRGB=false
   )
  ]
  rendererClasses [&NiagaraSpriteRendererProperties]
  renderers [{rendererClass=&NiagaraSpriteRendererProperties}]
 }
 Spike_and_Flare {
  emitterUpdateScript [
   &EmitterState()
   &SpawnBurst_Instantaneous(Spawn Count=1, Use Spawn Probability=false)
  ]
  particleSpawnScript [
   &InitializeParticle(
    Color=<rgba 1.75 1.6325 0.3895 0.75>
    Lifetime=0.2
    Mesh Scale=&Multiply_VectorByFloat(Vector=&UniformRangedVector(Minimum=<1.5 2 2>, Maximum=<2 3 3>), Float=1)
    Module.Write Color=true
    Module.Write Lifetime=true
    Module.Write Mass=true
    Module.Write Material Random=true
    Module.Write Position=true
    Module.Write Scale=true
    Module.Write SpriteRotation=false
    Module.Write SpriteSize=true
    Position=&Add_Vector(A=/Niagara/DynamicInputs/Helpers/SimulationPosition(), B=<0 0 0>)
    Sprite Size=&Vector2DFromFloat(Value=750)
   )
   &DynamicMaterialParameters(
    Index 0 Param 3=&UniformRangedFloat(Minimum=1, Maximum=1.5)
    Index 1 Param 1=0.2
    Index 1 Param 2=1
    Module.Param0WriteEnabled=false
    Module.Param0WriteEnabled1=true
    Module.Param1WriteEnabled=false
    Module.Param1WriteEnabled1=true
    Module.Param2WriteEnabled=true
    Module.Param2WriteEnabled1=true
    Module.Param3WriteEnabled=false
    Module.Param3WriteEnabled1=true
    Write Parameter Index 1=true
   ) {module=DynamicMaterialParameters001}
   &InitialMeshOrientation(Rotation=<0 0.25 0>, Use Orientation Vector=true, Use Rotation Vector=true)
   &CameraOffset(Camera Offset Amount=5)
  ]
  particleUpdateScript [
   &ParticleState()
   &ScaleMeshSize(Scale Factor=&VectorFromCurve(VectorCurve=DataInterface() {propertyValues {
       bOptimizeLUT=false
       XCurve {keys [
         {arriveTangent=1.8635, interpMode=RCIM_Cubic, time=0, value=0}
         {arriveTangent=1, interpMode=RCIM_Cubic, leaveTangent=1, time=0.5, value=0.73}
         <at 1 1 RCIM_Cubic>
        ]}
       YCurve {keys [
         {arriveTangent=0.1613, interpMode=RCIM_Cubic, time=0, value=0.2}
         <at 0.5 1 RCIM_Cubic>
         {arriveTangent=-5.6805, interpMode=RCIM_Cubic, leaveTangent=-5.6805, tangentMode=RCTM_User, time=1, value=0.2}
        ]}
       ZCurve {keys [
         <at 0 0.2 RCIM_Cubic>
         <at 0.5 1 RCIM_Cubic>
         {arriveTangent=-5.4621, interpMode=RCIM_Cubic, leaveTangent=-5.4621, tangentMode=RCTM_User, time=1, value=0.2}
        ]}
      }}))
   &ScaleColor(
    Scale Mode=::NewEnumerator1
    Scale RGBA=&MakeVector4_FromLinearColor(LinearColor=&ColorFromCurve(ColorCurve=DataInterface() {propertyValues {
        AlphaCurve {keys [<at 0 1 RCIM_Cubic>,<at 1 0 RCIM_Cubic>]}
        BlueCurve {keys [<at 0 5 RCIM_Cubic>,<at 1 0 RCIM_Cubic>]}
        GreenCurve {keys [<at 0 5 RCIM_Cubic>,<at 1 0 RCIM_Cubic>]}
        RedCurve {keys [<at 0 5 RCIM_Cubic>,<at 1 0 RCIM_Cubic>]}
       }}))
    ScaleRGBA=true
   )
   &DynamicMaterialParameters(
    Index 0 Param 1=&FloatFromCurve(FloatCurve=DataInterface() {propertyValues {Curve {keys [<at 0.1 0.1 RCIM_Cubic>,<at 0.8 1>]}}})
    Index 0 Param 2=2
    Index 0 Param 4=&FloatFromCurve(FloatCurve=DataInterface() {propertyValues {Curve {keys [<at 0 1>,<at 1 3>]}}})
    Module.Param0WriteEnabled=true
    Module.Param1WriteEnabled=true
    Module.Param2WriteEnabled=false
    Module.Param3WriteEnabled=true
   )
   /SetVariables(
    Particles.SpriteFlareCameraOffset=50
    Particles.SpriteFlareColor:LinearColor=&MakeLinearColorFromVectorAndFloat(
     "Vector (RGB)"=&MakeVectorFromLinearColorRGB(Color=<rgba 2 1.7936 0.8438 1>)
     "Float (Alpha)"=&Lerp_Float(A=0.05, B=0, Alpha=$Particles.NormalizedAge)
    )
   ) {isSetParameters=true}
  ]
  rendererClasses [&NiagaraMeshRendererProperties, &NiagaraSpriteRendererProperties]
  renderers [
   {rendererClass=&NiagaraMeshRendererProperties, rendererIndex=0}
   {rendererClass=&NiagaraSpriteRendererProperties, rendererIndex=1}
  ]
 }
}
summary {
 systemName=NS_Emote_LunchBox_Impact
 emitters [
  {emitterName=Light, rendererClasses [&NiagaraLightRendererProperties]}
  {
   emitterName=Spike_and_Flare
   rendererClasses [&NiagaraMeshRendererProperties, &NiagaraSpriteRendererProperties]
  }
  {bEnabled=false, emitterName=Electric_Strokes, rendererClasses [&NiagaraSpriteRendererProperties]}
  {emitterName=Lazy_Sparks, rendererClasses [&NiagaraSpriteRendererProperties]}
  {emitterName=ElectricGround, rendererClasses [&NiagaraSpriteRendererProperties]}
  {emitterName=SmokeGround, rendererClasses [&NiagaraSpriteRendererProperties]}
  {emitterName=Lightning_residue, rendererClasses [&NiagaraMeshRendererProperties]}
  {bEnabled=false, emitterName=ImpactShockWave_Energy, rendererClasses [&NiagaraMeshRendererProperties]}
  {emitterName=LightningSprites_Burst, rendererClasses [&NiagaraSpriteRendererProperties]}
 ]
 userVariables []
}
